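# Physics asset for the Medic backpack cosmetic, defining bodies and joints for four dangling ball-and-wheel chains on the pack.
# Each top and bottom, left and right chain links a box root, capsule ball and sphere wheel.
# Balls swing within limits, while wheels have mostly locked swing and rotate chiefly about one axis.
# Only the bottom roots are constrained to the backpack root, with all angular motion locked; the top roots have no such constraint.
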PhysicsAsset /BRCosmetics/Accessories/FORT_Backpacks/F_MED_Medic_Backpack/Meshes/F_MED_Medic_Physics
bodies {
 backpack_root [{shapeType=Capsule, length=2.059, radius=0.8904, center=<-5.8591 -9.9314 0.0272>, rotation=<rot 0.0624 -79.9996 -98.3477>}]
 dyn_ball_bot_l [{shapeType=Capsule, length=9.75, radius=4.75}]
 dyn_ball_bot_r [{shapeType=Capsule, length=11, radius=5.25}]
 dyn_ball_top_l [{shapeType=Capsule, length=4.5, radius=2.95}]
 dyn_ball_top_r [{shapeType=Capsule, length=4.9294, radius=2.9561}]
 dyn_root_bot_l [{shapeType=Box, extentX=4, extentY=4, extentZ=4}]
 dyn_root_bot_r [{shapeType=Box, extentX=3.5, extentY=3.5, extentZ=3.5}]
 dyn_root_top_l [{shapeType=Box, extentX=2, extentY=2, extentZ=2, center=<0 2.8311 0>}]
 dyn_root_top_r [{shapeType=Box, extentX=2, extentY=2, extentZ=2}]
 dyn_wheel_bot_l [{shapeType=Sphere, radius=3.5, center=<0 0 0.1>}]
 dyn_wheel_bot_r [{shapeType=Sphere, radius=3.25, center=<0 0 0.1>}]
 dyn_wheel_top_l [{shapeType=Sphere, radius=3.5}]
 dyn_wheel_top_r [{shapeType=Sphere, radius=3.75}]
}
constraints [
 {bone1Name=dyn_wheel_top_l, bone2Name=dyn_ball_top_l, swing1LimitDegrees=0, swing1Motion=Locked, swing2LimitDegrees=180, twistLimitDegrees=3}
 {
  bone1Name=dyn_wheel_top_r
  bone2Name=dyn_ball_top_r
  swing1LimitDegrees=0.01
  swing1Motion=Locked
  swing2LimitDegrees=180
  twistLimitDegrees=5
 }
 {bone1Name=dyn_ball_top_l, bone2Name=dyn_root_top_l, swing1LimitDegrees=60, swing2LimitDegrees=35, twistLimitDegrees=4}
 {bone1Name=dyn_ball_top_r, bone2Name=dyn_root_top_r, swing1LimitDegrees=60, swing2LimitDegrees=35, twistLimitDegrees=4}
 {bone1Name=dyn_ball_bot_l, bone2Name=dyn_root_bot_l, swing1LimitDegrees=15, swing2LimitDegrees=25, twistLimitDegrees=80}
 {
  bone1Name=dyn_wheel_bot_l
  bone2Name=dyn_ball_bot_l
  swing1LimitDegrees=0.1
  swing1Motion=Locked
  swing2LimitDegrees=2
  twistLimitDegrees=180
 }
 {
  bone1Name=dyn_wheel_bot_r
  bone2Name=dyn_ball_bot_r
  swing1LimitDegrees=45
  swing1Motion=Locked
  swing2LimitDegrees=45
  swing2Motion=Locked
  twistLimitDegrees=180
 }
 {
  bone1Name=dyn_root_bot_l
  bone2Name=backpack_root
  swing1LimitDegrees=45
  swing1Motion=Locked
  swing2LimitDegrees=45
  swing2Motion=Locked
  twistLimitDegrees=45
  twistMotion=Locked
 }
 {
  bone1Name=dyn_root_bot_r
  bone2Name=backpack_root
  swing1LimitDegrees=45
  swing1Motion=Locked
  swing2LimitDegrees=45
  swing2Motion=Locked
  twistLimitDegrees=45
  twistMotion=Locked
 }
 {bone1Name=dyn_ball_bot_r, bone2Name=dyn_root_bot_r, swing1LimitDegrees=20, swing2LimitDegrees=20, twistLimitDegrees=80}
]
