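# Anim Blueprint for the F_MED_Medic backpack, adding secondary physics motion that reacts to skydiving and parachuting.
# From the ref pose, four chained Anim Dynamics nodes spring the top and bottom left/right bones, followed by a RigidBody node.
# Update event turns pawn height into Perlin-noise external forces while skydiving, damped under parachute, feeding the Anim Dynamics nodes.
# Skydive and parachute flags also switch simulation alphas and force vector.
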AnimBlueprint /BRCosmetics/Accessories/FORT_Backpacks/F_MED_Medic_Backpack/Meshes/F_MED_Medic_Skeleton_AnimBlueprint
target_skeleton=&/F_MED_Medic_Skeleton
graphs [
 AnimGraph {nodes [
   n1 = Get bIsParachuteOpened()
   n2 = SelectVector(
    A=Get falling_external_force_y() -> (falling_external_force_y:Vector)
    B=Get falling_external_force_z() -> (falling_external_force_z:Vector)
    bPickA=@n1
   )
   n3 = Get bIsSkyDiving()
   n4 = SelectFloat(A=0.35, B=SelectFloat(A=0.65, B=0.15, bPickA=@n3), bPickA=@n1)
   n5 = Anim Dynamics - Bone: dyn_root_top_r(ComponentPose=Local To Component(LocalPose=Local Space Ref Pose()), Alpha=@n4, ExternalForce=@n2) {
    angularSpringConstant=9000
    bAngularSpring=true
    bGravityOverrideInSimSpace=true
    gravityScale=0.3
    alphaBoolBlend {blendOption=Linear}
    boundBone {boneName=dyn_root_top_r}
    componentLinearAccScale=<1 1 1>
    componentLinearVelScale=<1 1 1>
    physicsBodyDefinitions [{
      sphereCollisionRadius=0
      boundBone {boneName=dyn_root_top_r}
      boxExtents=<10 10 10>
      constraintSetup {angularLimitsMax=<45 0 45>, angularLimitsMin=<-45 0 -10>, angularTarget=<0 0 0>}
      localJointOffset=<3 4 1>
     }]
   }
   n6 = Anim Dynamics - Bone: dyn_root_top_l(ComponentPose=@n5, Alpha=@n4, ExternalForce=@n2) {
    angularSpringConstant=9000
    bAngularSpring=true
    bGravityOverrideInSimSpace=true
    gravityScale=0.4
    alphaBoolBlend {blendOption=Linear}
    boundBone {boneName=dyn_root_top_l}
    componentLinearAccScale=<1 1 1>
    componentLinearVelScale=<1 1 1>
    physicsBodyDefinitions [{
      sphereCollisionRadius=0
      boundBone {boneName=dyn_root_top_l}
      boxExtents=<10 10 10>
      constraintSetup {angularLimitsMax=<45 0 10>, angularLimitsMin=<-45 0 -45>, angularTarget=<0 0 0>}
      localJointOffset=<-3 4 1>
     }]
   }
   n7 = Anim Dynamics - Bone: dyn_root_bot_r(ComponentPose=@n6, Alpha=@n4, ExternalForce=@n2) {
    angularSpringConstant=9000
    bAngularSpring=true
    bGravityOverrideInSimSpace=true
    gravityScale=0.3
    alphaBoolBlend {blendOption=Linear}
    boundBone {boneName=dyn_root_bot_r}
    componentLinearAccScale=<1 1 1>
    componentLinearVelScale=<1 1 1>
    physicsBodyDefinitions [{
      sphereCollisionRadius=0
      boundBone {boneName=dyn_root_bot_r}
      boxExtents=<10 10 10>
      constraintSetup {twistAxis=AxisY, angularLimitsMax=<50 1 20>, angularLimitsMin=<-1 -1 5>, angularTarget=<0 0 0>}
      localJointOffset=<3 4 2>
     }]
   }
   n8 = Anim Dynamics - Bone: dyn_root_bot_l(ComponentPose=@n7, Alpha=@n4, ExternalForce=@n2) {
    angularSpringConstant=4500
    bAngularSpring=true
    bGravityOverrideInSimSpace=true
    gravityScale=0.3
    alphaBoolBlend {blendOption=Linear}
    boundBone {boneName=dyn_root_bot_l}
    componentLinearAccScale=<1 1 1>
    componentLinearVelScale=<1 1 1>
    physicsBodyDefinitions [{
      sphereCollisionRadius=0
      boundBone {boneName=dyn_root_bot_l}
      boxExtents=<10 10 10>
      constraintSetup {twistAxis=AxisY, angularLimitsMax=<45 0 5>, angularLimitsMin=<20 0 -35>, angularTarget=<0 0 0>}
      localJointOffset=<-3 4 -1>
     }]
   }
   n9 = RigidBody(ComponentPose=@n8, Alpha=SelectFloat(A=0.95, B=SelectFloat(A=0.35, B=1, bPickA=@n3), bPickA=@n1)) {
    overridePhysicsAsset=&/F_MED_Medic_Physics
    alphaBoolBlend {blendOption=Linear}
    componentAppliedLinearAccClamp=<1000 1000 1000>
    componentLinearAccScale=<1.15 1.15 1.15>
    componentLinearVelScale=<1 1 1>
    overrideWorldGravity=<1 1 1>
   }
   Output Pose(Result=Component To Local(ComponentPose=@n9)) {layerGroup=None}
  ]}
 EventGraph {nodes [
   Event BlueprintUpdateAnimation()
   n1 = Get PawnLastLocation() -> (PawnLastLocation:Vector)
   n2 = SelectFloat(
    A=Make Perlin Noise and Remap!(Value:float=BreakVector(InVec=@n1).Z, RangeOutMin:float=-13000, RangeOutMax:float=13000) -> (ReturnValue:float)
    B=0
    bPickA=Get bIsSkyDiving()
   )
   n3 = SelectFloat(A=float * float(A=@n2, B=0.65), B=@n2, bPickA=Get bIsParachuteOpened())
   > Set falling_external_force_z(falling_external_force_z:Vector=MakeVector(X=@n3, Y=0, Z=@n3)) -> (Output_Get:Vector)
   > Set falling_external_force_y(falling_external_force_y:Vector=MakeVector(X=@n3, Y=@n3, Z=@n3)) -> (Output_Get:Vector)
   TryGetPawnOwner()
   Get PawnVelocityZ()
  ]}
]
variables [
 falling_external_force_y {type=Vector}
 falling_external_force_z {type=Vector}
]
